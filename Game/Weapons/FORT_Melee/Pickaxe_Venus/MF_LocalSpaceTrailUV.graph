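# Material function for the Venus pickaxe that builds object-relative trail texture coordinates.
# A SphereMask between world position and object position, shaped by range and hardness inputs, becomes U as a distance falloff from the object.
# The UV output appends the mesh's original V to that U.
# A static bool input mirrors only the separate U output; a similarly named switch parameter is left unconnected.
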MaterialFunction /Game/Weapons/FORT_Melee/Pickaxe_Venus/MF_LocalSpaceTrailUV
graph {nodes [
  n1 = FunctionInput(Preview=Constant()) {bUsePreviewValueAsDefault=true, desc=Added By Cheng, inputName=RangeRampHardness, inputType=FunctionInput_Scalar, sortPriority=1}
  n2 = SphereMask(
   A=WorldPosition().XYZ
   B=ObjectPositionWS()
   Radius=FunctionInput(Preview=Constant() {r=600}) {bUsePreviewValueAsDefault=true, inputName=MaxiumURange, inputType=FunctionInput_Scalar}
   Hardness=@n1
  )
  n3 = TextureCoordinate()
  FunctionOutput(None=AppendVector(A=@n2, B=ComponentMask(None=@n3) {g=true})) {outputName="LocalSpaceCoordinate(UV)"}
  n4 = FunctionInput(Preview=StaticBool()) {bUsePreviewValueAsDefault=true, desc=Added By Cheng, inputName=MirrorTexCoordU, inputType=FunctionInput_StaticBool, sortPriority=2}
  FunctionOutput(None=StaticSwitch(True=OneMinus(None=@n2), False=@n2, Value=@n4) {desc=Added By Cheng}) {desc=Added By Cheng, outputName="LocalSpaceCoordinate(U)", sortPriority=1}
  StaticSwitchParameter() {desc=Added By Cheng, parameterName="MirrorTexCoord(U)"}
 ]}
