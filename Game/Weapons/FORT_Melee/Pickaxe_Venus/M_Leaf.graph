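# Masked, lit particle material for leaves, stored with the Venus pickaxe and usable on sprites and Niagara mesh particles.
# Switches optionally pick a random cell from a four-leaf texture strip and recolor it between two colors by green channel, before particle color tinting.
# A CameraShrink switch selects an offset, likely for vertices: camera-proximity shrinking, or a sine bend driven by a dynamic Time parameter, zeroed on mobile.
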Material /Game/Weapons/FORT_Melee/Pickaxe_Venus/M_Leaf
graph {
 properties {
  bUsedWithNiagaraMeshParticles=true
  bUsedWithNiagaraSprites=true
  bUsedWithParticleSprites=true
  blendMode=BLEND_Masked
  pixelDepthOffsetMode=PDOM_Legacy
  usedShadingModels=Default Lit
 }
 nodes [
  Add(A=TextureCoordinate() {uTiling=0.25}, B=Constant2Vector() {r=0.75})
  Constant()
  Constant() {r=0.5}
  Constant()
  n1 = VectorParameter() {parameterName=ColorLow, defaultValue=<rgba 0 0 0 1>}
  n2 = VectorParameter() {parameterName=ColorHigh, defaultValue=<rgba 0 0 0 1>}
  n3 = Constant() {r=1}
  n4 = Constant() {r=4}
  n5 = ParticleRandom()
  n6 = TextureSampleParameter2D(UVs=StaticSwitchParameter(
    True=MaterialFunctionCall!(UseTexCoord 1, Rows=@n3, Columns=@n4, Particle Random Value=@n5) {materialFunction=/Game/Effects/Fort_Effects/Materials/_Functions/MF_RandomSubUV_GPU_Sprites}.Result
    False=TextureCoordinate()
   ) {parameterName=UseRandomSubUVforGPUsprites}) {parameterName=BaseTexture, texture=/Game/Effects/Fort_Effects/World/Enviroment/Falling_Leaf/Textures/T_Leaves_Large_4x1}
  Multiply(
   A=StaticSwitchParameter(True=LinearInterpolate(A=@n1.RGB, B=@n2.RGB, Alpha=@n6.G), False=@n6.RGB) {parameterName=CustomColor}
   B=ParticleColor().RGB
  )
  ObjectPositionWS()
  n7 = Subtract(A=ObjectPositionWS(), B=WorldPosition().XYZ)
  n8 = Saturate(None=SphereMask(A=ParticlePositionWS(), B=CameraPositionWS(), Radius=Constant() {r=150}, Hardness=Constant() {r=0.1}))
  n9 = Transform(None=Constant3Vector() {constant=<rgba 0 1 0 0>})
  n10 = DynamicParameter() {paramNames [Time,Param2,Param3,Param4]}
  n11 = WorldPosition()
  n12 = ParticlePositionWS()
  n13 = Multiply(A=Multiply(A=Sine(None=@n10.Time), B=Subtract(A=@n11.XYZ, B=@n12))) {constB=-1}
  StaticSwitchParameter(
   True=LinearInterpolate(B=Multiply(A=@n7, B=@n8) {constB=0.5}, Alpha=@n8)
   False=ShadingPathSwitch(Default=Multiply(A=@n9, B=@n13), Mobile=Constant()) {desc=Migrated from FeatureLevelSwitch}
  ) {parameterName=CameraShrink}
 ]
}
